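# Niagara ribbon trail for the TechnoTrail vehicle boost cosmetic, tinted light blue or by the user painted color.
# A single emitter spawns one ribbon per filtered socket port, attached to skeletal positions with a car offset.
# Particles move along negative X and inherit source movement, which tapers late in life.
# Curves shape brightness and alpha fade, while randomized dynamic material parameters drive the ribbon material.
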NiagaraSystem /VehicleCosmetics/Boosts/Boost_TechnoTrail/FX/NS_Boost_TechnoTrail
dependencies {
 usedDataInterfaces [&NiagaraDataInterfaceCurve, &NiagaraDataInterfaceSkeletalMesh]
 usedDynamicInputs []
 usedModules [
  &self:SystemSpawnScript.SystemScriptSource.SystemScriptGraph.NiagaraNodeAssignment_0.SetVariables_#1
  &SystemState
 ]
}
emitters {NE_VC_Ribbon_Parent001 {
  emitterUpdateScript [
   &EmitterState(
    Inactive Response=::NewEnumerator1
    Life Cycle Mode=::NewEnumerator1
    Loop Behavior=::NewEnumerator1
    Loop Duration Mode=::NewEnumerator1
    UseLoopDelay=false
   )
   &VC_FilteredSocketCount()
   &VC_RateQualityHandler(Use Custom Rate?=true)
   &VC_SpawnRibbon_v1(Number of Ribbons=$Emitter.PortCount, SpawnRate=60)
  ]
  particleSpawnScript [
   &VC_PaintedVariantsOverride() {enabled=false}
   &VC_InitializeRibbon_v1(
    Custom Offset=true
    Enable Skeletal Position=true
    Lifetime=0.8
    Local Vector Offset=&AddVectorToPosition(Position=<0 0 0>, Vector=$User.DTCarOffset)
    Offset Amount=1
    Ribbon Width=60
   )
   &AddVelocity(Velocity=<-400 0 0>) {enabled=false, module=AddVelocity001}
   &AddVelocity(Velocity=<1500 0 0>, Velocity Speed Scale=-1)
   &Color(Color=&MakeCustomLinearColorFromBool(Bool=$User.bIsPainted, True Color=$User.PaintedColor, False Color=<rgba 0.1406 0.6181 1 1>)) {module=Color001}
  ]
  particleUpdateScript [
   &ParticleState()
   &InheritSourceMovement(
    Applied Position Scale=&Multiply_VectorByFloat(
     Vector=<1 1 1>
     Float=&FloatFromCurve(FloatCurve=DataInterface() {propertyValues {Curve {keys [<at 0.5 1 RCIM_Cubic>,<at 1 0 RCIM_Cubic>]}}})
    )
    ApplyForce=false
    ApplyPosition=true
    ApplyPositionFalloff=false
   )
   &SolveForcesAndVelocity(Clamp Velocity=false, Limit Acceleration=false)
   &VC_UpdateRibbon_v1(Enable Skeletal Position=true)
   &DynamicMaterialParameters(
    Index 0 Param 1=&RandomRangeFloat(Minimum=5, Maximum=10)
    Index 0 Param 2=&RandomRangeFloat(Minimum=1.5, Maximum=3)
    Index 0 Param 3=&RandomRangeFloat(Minimum=1.5, Maximum=3)
    Index 0 Param 4=&RandomRangeFloat(Minimum=1.5, Maximum=2)
    Index 1 Param 1=&MakeCustomFloatFromBool(Bool=$User.bIsFrontendPreview, True Float=0.2, False Float=0.2)
    Module.Param0WriteEnabled=true
    Module.Param0WriteEnabled1=true
    Module.Param1WriteEnabled=true
    Module.Param1WriteEnabled1=true
    Module.Param2WriteEnabled=true
    Module.Param2WriteEnabled1=true
    Module.Param3WriteEnabled=true
    Module.Param3WriteEnabled1=true
    Write Parameter Index 1=true
   ) {enabled=false}
   &ScaleColor(
    Scale Alpha=&FloatFromCurve(FloatCurve=DataInterface() {propertyValues {Curve {keys [
         <at 0 2 RCIM_Cubic>
         {arriveTangent=0.2423, interpMode=RCIM_Cubic, leaveTangent=0.2423, tangentMode=RCTM_User, time=0.7, value=1}
         <at 1 0 RCIM_Cubic>
        ]}}})
    Scale RGB=&VectorFromFloat(Value=&FloatFromCurve(FloatCurve=DataInterface() {propertyValues {Curve {keys [<at 0 1.5 RCIM_Cubic>,<at 0.25 1 RCIM_Cubic>,<at 0.6 1.5 RCIM_Cubic>]}}}))
    ScaleA=true
    ScaleRGB=true
   )
   &DynamicMaterialParameters(
    Index 0 Param 1=&RandomRangeFloat(Minimum=5, Maximum=10)
    Index 0 Param 2=&RandomRangeFloat(Minimum=1.5, Maximum=3)
    Index 0 Param 3=&RandomRangeFloat(Minimum=1.5, Maximum=3)
    Index 0 Param 4=&RandomRangeFloat(Minimum=1.5, Maximum=2)
    Index 1 Param 1=&MakeCustomFloatFromBool(Bool=$User.bIsFrontendPreview, True Float=0.2, False Float=0.2)
    Module.Index0Float0Write=true
    Module.Index0Float1Write=true
    Module.Index0Float2Write=true
    Module.Index0Float3Write=true
    Module.Index1Float0Write=true
    Module.Index1Float1Write=true
    Module.Index1Float2Write=true
    Module.Index1Float3Write=true
    Write Parameter Index 1=::NewEnumerator1
   ) {module=DynamicMaterialParameters001}
  ]
  rendererClasses [&NiagaraRibbonRendererProperties]
  renderers [{rendererClass=&NiagaraRibbonRendererProperties}]
 }}
summary {
 systemName=NS_Boost_TechnoTrail
 emitters [{emitterName=NE_VC_Ribbon_Parent001, rendererClasses [&NiagaraRibbonRendererProperties]}]
 userVariables [
  User.bIsFrontendPreview {type=NiagaraBool}
  User.bIsPainted {type=NiagaraBool}
  User.BlueCh {type=LinearColor, description="", defaultValue {struct=LinearColor, value=<rgba 0.1412 0.5686 0.8275 1>}}
  User.ColorVariantID {type=NiagaraInt32}
  User.DTCarOffset {type=Vector3f, description="", defaultValue {struct=Vector3f, value=<0 0 0>}}
  User.GreenCh {type=LinearColor, description="", defaultValue {struct=LinearColor, value=<rgba 0.1176 0.1255 1 1>}}
  User.PaintedColor {type=LinearColor, description="", defaultValue {struct=LinearColor, value=<rgba 1 1 1 1>}}
  User.PaintedScalar {type=NiagaraFloat, defaultValue=1, description=""}
  User.PortCount {type=NiagaraInt32, defaultValue=2}
  User.RedCh {type=LinearColor, description="", defaultValue {struct=LinearColor, value=<rgba 1 0.149 0.051 1>}}
 ]
}
