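# Unlit translucent Niagara ribbon material for the TechnoTrail boost's light trail, tinted by particle color.
# A gradient stripe splits bright trim from a core of three panning line channels, colored by mask-blended parameters and boosted by dynamic-parameter highlights.
# Self-warped water noise creates distortion outside the core.
# Opacity combines stripe, lines, square mask and particle alpha, with a dynamic MaskFade power fade along the U axis.
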Material /VehicleCosmetics/Boosts/Boost_TechnoTrail/Material/M_LightTrailPaintable
graph {
 properties {
  bUsedWithNiagaraRibbons=true
  blendMode=BLEND_Translucent
  pixelDepthOffsetMode=PDOM_Legacy
  refractionMethod=RM_PixelNormalOffset
  shadingModel=MSM_Unlit
  usedShadingModels=Unlit
 }
 nodes [
  n1 = ScalarParameter() {defaultValue=25, parameterName=LinesHighlight}
  n2 = ScalarParameter() {defaultValue=10, parameterName=TrimHighlights}
  n3 = Panner(Coordinate=TextureCoordinate() {uTiling=0.5, vTiling=0.2}) {speedX=-2, speedY=0.01}
  n4 = TextureSample(UVs=@n3) {texture=/DealershipSharedLibrary/FX/Textures/TexturePackage/T_VCVFX_TechnoTrailsHL_RGB}
  n5 = Multiply(
   A=Add(
    A=Multiply(A=@n2, B=@n4.R)
    B=TextureSample(UVs=@n3) {texture=/DealershipSharedLibrary/FX/Textures/TexturePackage/T_VCVFX_TechnoTrails_RGB}.R
   )
   B=ScalarParameter() {defaultValue=10, parameterName=TrimBrightness}
  )
  n6 = TextureSample() {texture=/DealershipSharedLibrary/FX/Textures/Gradient_Masks/T_VCVFX_GradientStripe_A}
  n7 = Saturate(None=ConstantBiasScale(None=@n6.RGB) {bias=-0.0015, scale=5})
  n8 = Saturate(None=OneMinus(None=@n7))
  n9 = Constant() {r=1.5}
  n10 = ParticleColor()
  n11 = LinearInterpolate(A=Multiply(A=Multiply(A=@n5, B=@n8), B=Multiply(A=@n9, B=@n10.RGB)), B=Multiply(A=Constant() {r=0.3}, B=Saturate(None=@n10.RGB)), Alpha=@n7)
  n12 = DynamicParameter() {paramNames [Highlight01, Highlight02, Highlight04, CoreBrightness]}
  n13 = Panner(Coordinate=TextureCoordinate() {uTiling=3, vTiling=0.3}, Time=Multiply(A=Constant() {r=1}, B=Time())) {speedX=-2.5, speedY=0.01}
  n14 = ParticleColor()
  n15 = TextureSample() {texture=/DealershipSharedLibrary/FX/Textures/TexturePackage/T_VCVFX_Masks_RGB}
  n16 = Multiply(A=Constant() {r=2}, B=Power(Base=@n15.G, Exp=Constant() {r=2}))
  n17 = VectorParameter() {parameterName=RedCh, defaultValue=<rgba 0 0.2098 1 2>}
  n18 = Multiply(A=@n17.RGB, B=@n17.RGB)
  n19 = TextureSample(UVs=@n13) {texture=&T_VCVFX_TechnoTrails_RGB}
  n20 = Add(
   A=Multiply(A=@n12.Highlight01, B=TextureSample(UVs=@n13) {texture=&T_VCVFX_TechnoTrailsHL_RGB}.R)
   B=Multiply(A=LinearInterpolate(A=@n18, B=@n14.RGB, Alpha=@n16) {constB=0}, B=@n19.R)
  )
  n21 = Panner(Coordinate=TextureCoordinate() {uTiling=2, vTiling=0.2}) {speedX=-2, speedY=-0.01}
  n22 = TextureSample(UVs=@n21) {texture=&T_VCVFX_TechnoTrailsHL_RGB}
  n23 = VectorParameter() {parameterName=BlueCh, defaultValue=<rgba 0.1807 1 0 2>}
  n24 = Multiply(A=@n23.RGB, B=@n23.RGB)
  n25 = VectorParameter() {parameterName=GreenCh, defaultValue=<rgba 1 0.0984 0 0.5>}
  n26 = LinearInterpolate(A=@n24, B=Multiply(A=@n25.RGB, B=@n25.RGB), Alpha=@n15.A)
  n27 = TextureSample(UVs=@n21) {texture=&T_VCVFX_TechnoTrails_RGB}
  n28 = Multiply(A=@n6.RGB, B=Add(A=Multiply(A=@n12.Highlight02, B=@n22.G), B=Multiply(A=@n26, B=@n27.G)))
  n29 = DynamicParameter() {paramNames [Highlight01, Highlight02, Highlight04, CoreBrightness]}
  n30 = Panner(Coordinate=TextureCoordinate() {uTiling=0.5}) {speedX=-2, speedY=0.01}
  n31 = TextureSample(UVs=@n30) {texture=&T_VCVFX_TechnoTrails_RGB}
  n32 = Add(
   A=Multiply(A=@n29.Highlight04, B=TextureSample(UVs=@n30) {texture=&T_VCVFX_TechnoTrailsHL_RGB}.B)
   B=Multiply(A=LinearInterpolate(A=@n17.RGB, B=@n24, Alpha=@n15.B), B=@n31.B)
  )
  n33 = Multiply(A=Add(A=Add(A=@n20, B=@n28), B=Multiply(A=@n15.G, B=@n32)), B=@n7)
  n34 = Add(A=Add(A=@n19.R, B=@n27.G), B=@n31.B)
  n35 = Saturate(None=ConstantBiasScale(None=@n34) {bias=-0.0125, scale=5})
  n36 = Multiply(A=@n10.RGB, B=LinearInterpolate(A=Multiply(A=@n1, B=@n11), B=Multiply(A=@n33, B=@n29.CoreBrightness), Alpha=Multiply(A=@n7, B=@n35)))
  n37 = TextureSample() {texture=/DealershipSharedLibrary/FX/Textures/TexturePackage/T_VCVFX_SquareMask_RGB}
  MaterialFunctionCall(Emissive=Multiply(A=ScalarParameter() {defaultValue=1, parameterName=Temp_Emissive_power}, B=Multiply(A=@n36, B=@n37.R))) {materialFunction=&MF_FX_EmissiveAdjustments}
  n38 = OneMinus(None=@n7)
  n39 = TextureCoordinate() {uTiling=0.5, vTiling=0.5}
  n40 = TextureSample(UVs=Panner(Coordinate=@n39) {speedX=0.1, speedY=0.1}) {texture=/DealershipSharedLibrary/FX/Textures/Noise/T_VCVFX_Water01_A}
  n41 = Time()
  n42 = Constant() {r=6}
  n43 = TextureSample(UVs=Panner(Coordinate=@n39, Time=Multiply(A=@n41, B=@n42)) {speedX=-0.1, speedY=-0.1}) {texture=&T_VCVFX_Water01_A}
  n44 = Add(A=TextureCoordinate(), B=Multiply(A=AppendVector(A=@n40.R, B=@n43.G), B=Constant() {r=0.1}))
  Multiply(
   A=Multiply(A=Saturate(None=@n38), B=TextureSample(UVs=@n44) {texture=&T_VCVFX_Water01_A}.RGB)
   B=ScalarParameter() {defaultValue=3, parameterName=Distortion}
  )
  n45 = Multiply(A=Multiply(A=@n6.RGB, B=ConstantBiasScale(None=@n34) {bias=0.8, scale=1}), B=@n37.G)
  n46 = TextureCoordinate()
  Multiply(
   A=Saturate(None=Multiply(A=@n10.A, B=@n45))
   B=Power(Base=ComponentMask(None=@n46) {r=true}, Exp=DynamicParameter() {parameterIndex=1, paramNames [MaskFade,Param2,Param3,Param4]}.MaskFade) {constExponent=0.2}
  )
 ]
}
